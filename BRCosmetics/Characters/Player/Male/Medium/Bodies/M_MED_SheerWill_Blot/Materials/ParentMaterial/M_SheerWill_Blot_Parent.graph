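# Masked parent material for the Sheer Will Blot player character body, built on material attributes for skinned meshes with morph targets and mesh deformers.
# The shared 2023 character PreFX function feeds a Sheer Will emote blend opacity function, whose result enters the shared PostFX function.
# Named reroutes label material and emissive hand-off points between stages.
# The shading model comes from the expression graph, covering lit, subsurface, clear coat and cloth.
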Material /BRCosmetics/Characters/Player/Male/Medium/Bodies/M_MED_SheerWill_Blot/Materials/ParentMaterial/M_SheerWill_Blot_Parent
graph {
 properties {
  bAutomaticallySetUsageInEditor=false
  bUseMaterialAttributes=true
  bUsedWithMeshDeformer=true
  bUsedWithMorphTargets=true
  bUsedWithSkeletalMesh=true
  blendMode=BLEND_Masked
  pixelDepthOffsetMode=PDOM_Legacy
  shadingModel=MSM_FromMaterialExpression
  usedShadingModels=Default Lit | Subsurface | Clear Coat | Subsurface Profile | Cloth
 }
 nodes [
  n1 = MaterialFunctionCall() {materialFunction=&MF_CharacterParent_2023_PreFX}
  n2 = MaterialFunctionCall!(In=NamedRerouteDeclaration(None=@n1.Result) {name=Material In, nodeColor=<rgba 1 0.4219 0 1>}) {materialFunction=/BRCosmetics/Characters/Player/Male/Medium/Bodies/M_MED_SheerWill_Blot/Skins/FrothRain/Materials/MF/MF_SheerWillEmoteBlendMaterial_opacity}
  MaterialFunctionCall(Material=@n2.Result, M=@n1.M, Pre Skinned Local Position=@n1.M, Distance Blend=@n1.M, Fresnel=@n1.M) {materialFunction=&MF_CharacterParent_2023_PostFX}
  n3 = NamedRerouteUsage()
  NamedRerouteDeclaration(None=@n3) {name=Material Out, nodeColor=<rgba 1 0 0.9609 1>}
  NamedRerouteDeclaration(None=GetMaterialAttributes(None=@n3) {attributeGetTypes [#1]}.Emissive Color) {name=Emissive In, nodeColor=<rgba 0 1 0.7188 1>}
 ]
}
